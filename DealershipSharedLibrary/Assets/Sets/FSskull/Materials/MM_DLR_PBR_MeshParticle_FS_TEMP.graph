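# Masked, lit PBR master material for Niagara mesh particles in the FSskull dealership set, marked TEMP, with full-screen transition alpha support.
# Panning base color, normal and packed SRM maps combine with particle color and dynamic parameters for paint, emission, metallic and dissolve.
# Emissive blends texture, fresnel and light-vector fake shadows with an erosion-edge highlight.
# Switchable opacity factors feed depth fade and temporal dither; noise drives normal-direction WPO with camera offset.
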Material /DealershipSharedLibrary/Assets/Sets/FSskull/Materials/MM_DLR_PBR_MeshParticle_FS_TEMP
graph {
 properties {
  bAutomaticallySetUsageInEditor=false
  bUsedWithNiagaraMeshParticles=true
  blendMode=BLEND_Masked
  previewMesh=/Engine/Transient.MaterialEditorMeshComponent_17
  usedShadingModels=Default Lit
 }
 nodes [
  MaterialFunctionCall(Emissive=NamedRerouteUsage()) {materialFunction=&MF_FX_EmissiveAdjustments}
  n1 = TextureCoordinate()
  n2 = ChannelMaskParameter(None=MaterialFunctionCall!(UV Coord=@n1, LenghtScale, RadialScale) {materialFunction=/Game/Global/MaterialFunctions/TechArt/MF_PolarCoordinates}.Result) {group=06 WPO, parameterName=WPODirection}
  n3 = ScalarParameter() {defaultValue=2, group=06 WPO, parameterName=WPO Spread}
  n4 = NamedRerouteUsage()
  Multiply(
   A=Multiply(A=Power(Base=@n2.R, Exp=@n3), B=ScalarParameter() {defaultValue=1, group=06 WPO, parameterName=WPO Strength})
   B=Multiply(A=VertexNormalWS(), B=@n4)
  )
  n5 = VectorParameter() {
   group=01 Base Color
   parameterName=Base Color - UVs
   channelNames {a=Panner Y, b=Panner X, g=Scale Y, r=Scale X}
   defaultValue=<rgba 1 1 0 0>
  }
  n6 = Multiply(A=TextureCoordinate(), B=ComponentMask(None=@n5) {g=true, r=true})
  n7 = Multiply(A=AppendVector(A=@n5.Panner X, B=@n5.Panner X), B=Time())
  n8 = TextureSampleParameter2D(UVs=NamedRerouteDeclaration(None=Add(A=@n6, B=@n7)) {name=BC UVs, nodeColor=<rgba 0.7344 0 1 1>}) {group=01 Base Color, parameterName=Base Color Texture, sortPriority=1, texture=/DealershipSharedLibrary/Assets/Sets/Lava/T_Lava_Cracks_D_E}
  n9 = Multiply(
   A=StaticSwitchParameter(True=ChannelMaskParameter(None=@n8.RGB) {parameterName=Base - Channel}.R, False=@n8.RGB) {defaultValue=true, parameterName=Base Color - Select Channel}
   B=VectorParameter() {group=01 Base Color, parameterName=Base Color - Tint, defaultValue=<rgba 1 1 1 1>}.RGB
  )
  n10 = Desaturation(None=@n9, Fraction=Constant() {r=1})
  n11 = ParticleColor()
  NamedRerouteDeclaration(None=LinearInterpolate(A=@n9, B=Multiply(A=@n10, B=@n11.RGB), Alpha=NamedRerouteUsage())) {name=Base Color, nodeColor=<rgba 1 0.2001 0 1>}
  n12 = TextureSampleParameter2D(UVs=NamedRerouteUsage()) {
   group=05 Normal
   parameterName=Normal - Texture
   samplerType=SAMPLERTYPE_Normal
   sortPriority=0
   texture=&FlatNormalBlue
  }
  NamedRerouteDeclaration(None=LinearInterpolate(
    A=Constant3Vector() {constant=<rgba 0 0 1 0>}
    B=@n12.RGB
    Alpha=ScalarParameter() {defaultValue=1, group=05 Normal, parameterName=Normal - Intensity}
   )) {name=Normal, nodeColor=<rgba 0.5 0.5 1 1>}
  n13 = DynamicParameter() {defaultValue=<rgba 0 1 0 0>, paramNames [Is Painted, Emissive, Metallic, Dissolve]}
  NamedRerouteDeclaration(None=@n13.Is Painted) {name=DP Is Painted, nodeColor=<rgba 0.7344 0 1 1>}
  NamedRerouteDeclaration(None=@n8.A) {name=Alpha, nodeColor=<rgba 0.4844 0.4844 0.4844 1>}
  n14 = NamedRerouteUsage()
  n15 = StaticSwitchParameter(True=Multiply(A=NamedRerouteUsage(), B=@n14), False=@n14) {
   defaultValue=true
   desc="Uses Base Texture's A channel for opacity mask by default (true)"
   group=04 Opacity
   parameterName=Opacity Mask-UseBaseTextureAlpha
  }
  n16 = StaticSwitchParameter(
   True=Multiply(
    A=MaterialFunctionCall!(Use MPC, Is Sprite, Use World Position Behind Translucency) {materialFunction=/DealershipSharedLibrary/FX/Materials/MaterialFunctions/MF_DLR_FullScreenTransition_Alpha}.Curve - Alpha
    B=@n15
   )
   False=@n15
  ) {
   defaultValue=true
   desc="Uses Base Texture's A channel for opacity mask by default (true)"
   group=04 Opacity
   parameterName=Opacity Mask-UseFullScreenTransition
  }
  n17 = StaticSwitchParameter(True=Multiply(A=@n16, B=ParticleColor().A), False=@n16) {
   desc="Uses Base Texture's A channel for opacity mask by default (true)"
   group=04 Opacity
   parameterName=Opacity Mask- Multiply Particle Alpha with Opacity
  }
  NamedRerouteDeclaration(None=MaterialFunctionCall(Opacity In=@n17, DepthFade Distance=ScalarParameter() {group=04 Opacity, parameterName=Opacity - Depth Fade}) {materialFunction=&MF_VCVFX_DepthFade}.Result) {name=Opacity, nodeColor=<rgba 1 1 1 1>}
  n18 = ScalarParameter() {defaultValue=5, group=03 Emissive, parameterName=Emissive - Fresnel Exp, sortPriority=10}
  NamedRerouteDeclaration(None=LinearInterpolate(
    A=ScalarParameter() {group=03 Emissive, parameterName=Emissive - Fresnel Min, sortPriority=11}
    B=ScalarParameter() {group=03 Emissive, parameterName=Emissive - Fresnel Max, sortPriority=12}
    Alpha=Fresnel(ExponentIn=@n18) {baseReflectFraction=0}
   )) {name=Fresnel, nodeColor=<rgba 0.5122 0.2671 0.5521 20>}
  n19 = DynamicParameter() {parameterIndex=1, defaultValue=<rgba 0 1 1 1>, paramNames [WPO, Base Color Emission Amount, Param3, Param4]}
  NamedRerouteDeclaration(None=@n19.WPO) {name=DP WPO, nodeColor=<rgba 0.7344 0 1 1>}
  n20 = MaterialFunctionCall(Offset Amount=ScalarParameter() {group=WPO, parameterName=WPO - CameraOffset}) {materialFunction=/Engine/Functions/Engine_MaterialFunctions02/WorldPositionOffset/CameraOffset}
  n21 = ScalarParameter() {defaultValue=-200, group=WPO, parameterName=WPO - Min Amount}
  n22 = ScalarParameter() {defaultValue=400, group=WPO, parameterName="WPO -Max  Amount"}
  n23 = VectorParameter() {
   group=WPO
   parameterName="WPO -  Noise UV Controls"
   channelNames {a=Panner V, b=Panner U, g=Tiling V, r=Tiling U}
   defaultValue=<rgba 1 1 0 0>
  }
  n24 = AppendVector(A=@n23.Tiling U, B=@n23.Tiling U)
  n25 = TextureSampleParameter2D(UVs=Panner(Coordinate=Multiply(A=@n1, B=@n24) {constB=2}, Speed=AppendVector(A=@n23.Panner U, B=@n23.Panner U)) {speedX=0.2, speedY=0.15}) {group=WPO, parameterName=WPO Noise, samplerType=SAMPLERTYPE_LinearColor, texture=/Game/Effects/SnowSandInteraction/Textures/SSI_Perlin}
  n26 = Multiply(A=VertexNormalWS(), B=Multiply(A=LinearInterpolate(A=@n21, B=@n22, Alpha=@n25.R) {constA=-0.3, constB=0.3}, B=@n4))
  NamedRerouteDeclaration(None=LinearInterpolate(A=Add(A=@n20.Clamped Camera Offset, B=@n26), B=Constant3Vector(), Alpha=ScalarParameter() {parameterName=UseWPO}) {constAlpha=0}) {name=WPO, nodeColor=<rgba 0.5938 0 1 1>}
  n27 = ScalarParameter() {defaultValue=2, group=03 Emissive, parameterName=Emissive - FakeShadow SmoothStep}
  n28 = LinearInterpolate(
   A=AtmosphericLightVector()
   B=VectorParameter() {parameterName=Emissive - FakeShadow CustomVectorLightDirection, defaultValue=<rgba 0 0 0 1>}.RGB
   Alpha=ScalarParameter() {group=03 Emissive, parameterName=Emissive - HasCustomLightDirection?, sliderMax=1}
  ) {constAlpha=0}
  n29 = VertexNormalWS()
  n30 = Power(
   Base=Add(A=DotProduct(A=@n28, B=@n29), B=ScalarParameter() {defaultValue=0.3, group=03 Emissive, parameterName=Emissive - FakeShadow MoveGradient})
   Exp=ScalarParameter() {defaultValue=1, group=03 Emissive, parameterName=Emissive - FakeShadow Power}
  )
  n31 = Saturate(None=Multiply(A=SmoothStep(Max=@n27, Value=@n30), B=ScalarParameter() {defaultValue=1, group=03 Emissive, parameterName=Emissive - FakeShadow Multiply}))
  n32 = VectorParameter() {group=03 Emissive, parameterName=Emissive - FakeShadow DarkColor, defaultValue=<rgba 0.0095 0.0115 0.0521 1>}
  n33 = VectorParameter() {group=03 Emissive, parameterName=Emissive - FakeShadow LightColor, defaultValue=<rgba 1 1 1 0>}
  NamedRerouteDeclaration(None=LinearInterpolate(
    A=@n31
    B=LinearInterpolate(A=@n32.RGB, B=@n33.RGB, Alpha=@n31)
    Alpha=ScalarParameter() {group=03 Emissive, parameterName=Emissive - ShadowsHaveColor?, sliderMax=1}
   )) {name=FakeShadow, nodeColor=<rgba 0 0.5078 1 1>}
  NamedRerouteDeclaration(None=MaterialFunctionCall(Alpha Threshold=@n17) {materialFunction=&DitherTemporalAA}.Result) {name=Masked Opacity, nodeColor=<rgba 0.4348 0.4348 0.4348 1>}
  n34 = TextureCoordinate() {coordinateIndex=1}
  n35 = If(A=ScalarParameter() {group=04 Opacity, parameterName=Dissolve - UV Channel}, "A > B"=@n34, "A == B"=@n34, "A < B"=TextureCoordinate())
  n36 = VectorParameter() {
   group=04 Opacity
   parameterName=Dissolve - UVs
   channelNames {a=Panner Y, b=Panner X, g=Scale Y, r=Scale X}
   defaultValue=<rgba 2.1167 1 0 0>
  }
  n37 = ComponentMask(None=@n36) {g=true, r=true}
  n38 = AppendVector(A=@n36.Panner X, B=@n36.Panner X)
  n39 = Time()
  n40 = TextureSampleParameter2D(UVs=Add(A=Multiply(A=@n35, B=@n37), B=Multiply(A=@n38, B=@n39))) {
   group=04 Opacity
   parameterName=Dissolve - Texture
   samplerType=SAMPLERTYPE_LinearColor
   sortPriority=30
   texture=/Game/Global/Textures/Default/Noise/T_CloudNoise_01
  }
  n41 = ChannelMaskParameter(None=@n40.RGBA) {group=04 Opacity, parameterName=Dissolve - Channel, sortPriority=31}
  n42 = MaterialFunctionCall!(
   Erosion=NamedRerouteUsage()
   Mask=@n41.R
   Edge Width=ScalarParameter() {group=04 Opacity, parameterName="Dissolve -  Edge Width"}
   Width=Constant() {r=0.25}
  ) {materialFunction=/DealershipSharedLibrary/FX/Materials/MaterialFunctions/MF_DLR_Erode}
  NamedRerouteDeclaration(None=@n42.Edge) {name=Dissolve Edge, nodeColor=<rgba 0.7813 0 1 1>}
  n43 = TextureSampleParameter2D(UVs=NamedRerouteUsage()) {
   group=02 SRM
   parameterName=SRM - Texture
   samplerType=SAMPLERTYPE_Masks
   sortPriority=0
   texture=&T_BlackMask
  }
  NamedRerouteDeclaration(None=ChannelMaskParameter(None=@n43.RGBA) {group=02 SRM, parameterName=Specular - Channel, sortPriority=1}.R) {name=Specular, nodeColor=<rgba 0.0122 0.0881 1 1>}
  NamedRerouteDeclaration(None=@n19.Base Color Emission Amount) {name=DP Base Color Emission Amount, nodeColor=<rgba 0.681 0 1 1>}
  NamedRerouteDeclaration(None=@n8.RGBA) {name=Base Texture, nodeColor=<rgba 0 0.1797 1 1>}
  n44 = ChannelMaskParameter(None=@n43.RGBA) {group=02 SRM, maskChannel=Green, parameterName=Roughness - Channel, sortPriority=2, defaultValue=<rgba 0 1 0 0>}
  NamedRerouteDeclaration(None=@n44.R) {name=Roughness, nodeColor=<rgba 0.0122 0.0881 1 1>}
  n45 = ChannelMaskParameter(None=@n43.RGBA) {group=02 SRM, maskChannel=Blue, parameterName=Metallic - Channel, sortPriority=3, defaultValue=<rgba 0 0 1 0>}
  NamedRerouteDeclaration(None=LinearInterpolate(A=@n45.R, Alpha=NamedRerouteUsage())) {name=Metallic, nodeColor=<rgba 0.0122 0.0881 1 1>}
  NamedRerouteDeclaration(None=@n42.Result) {name=Dissolve, nodeColor=<rgba 0.1875 0.6953 1 1>}
  n46 = VectorParameter() {group=04 Opacity, parameterName=Opacity - Dissolve Highlight Color, defaultValue=<rgba 1 0.5 0.1 1>}
  n47 = ScalarParameter() {defaultValue=1024, group=04 Opacity, parameterName=Opacity - Dissolve Highlight Gain}
  n48 = TextureCoordinate()
  n49 = VectorParameter() {group=03 Emissive, parameterName=Emissive - UVs, channelNames {a=Panner Y, b=Panner X, g=Scale Y, r=Scale X}, defaultValue=<rgba 1 1 0 0>}
  n50 = ComponentMask(None=@n49) {g=true, r=true}
  n51 = AppendVector(A=@n49.Panner X, B=@n49.Panner X)
  n52 = Time()
  n53 = TextureSampleParameter2D(UVs=Add(A=Multiply(A=@n48, B=@n50), B=Multiply(A=@n51, B=@n52))) {group=03 Emissive, parameterName=Emissive - Texture, sortPriority=1, texture=&T_Lava_Cracks_D_E}
  n54 = ChannelMaskParameter(None=If(
    A=ScalarParameter() {parameterName="Emissive- Use Base Texture (0 No 1 Yes)"}
    "A > B"=NamedRerouteUsage()
    "A == B"=@n53.RGBA
    "A < B"=@n53.RGBA
   )) {group=03 Emissive, maskChannel=Green, parameterName=Emissive - Channel, sortPriority=3, defaultValue=<rgba 0 1 0 0>}
  n55 = NamedRerouteUsage()
  n56 = ScalarParameter() {defaultValue=1, parameterName=Emissive Color Fresnel, sliderMax=1}
  n57 = Add(
   A=StaticSwitchParameter(True=@n54.R, False=Constant()) {defaultValue=true, group=03 Emissive, parameterName=UseEmissiveTexture, sortPriority=0}
   B=Multiply(A=@n55, B=@n56)
  )
  n58 = ParticleColor()
  n59 = LinearInterpolate(A=Multiply(A=@n55, B=VectorParameter() {parameterName=FresnelColor, defaultValue=<rgba 1 1 1 1>}.RGB), Alpha=@n56) {constB=0}
  n60 = Multiply(A=Add(A=Multiply(A=@n57, B=@n58.RGB), B=Clamp(None=@n59) {maxDefault=2}), B=NamedRerouteUsage())
  n61 = Multiply(
   A=NamedRerouteUsage()
   B=ScalarParameter() {defaultValue=1, group=03 Emissive, parameterName=Emissive - BaseColor Multiplier, sortPriority=15}
  )
  n62 = NamedRerouteUsage()
  n63 = LinearInterpolate(
   A=@n60
   B=Add(A=@n60, B=Multiply(A=@n61, B=@n62))
   Alpha=Round(None=ScalarParameter() {desc="0=False 1=True", group=03 Emissive, parameterName=AddBaseColorToEmissive?, sliderMax=1, sortPriority=13})
  ) {desc=Add base color to emissive}
  n64 = Saturate(None=LinearInterpolate(B=@n41.R, Alpha=ScalarParameter() {parameterName="Dissolve Emission? (0 No 1 Yes)"}) {constA=1})
  n65 = ScalarParameter() {parameterName=Dissolve Emission}
  n66 = Multiply(
   A=LinearInterpolate(A=Constant() {r=1}, B=NamedRerouteUsage(), Alpha=ScalarParameter() {group=03 Emissive, parameterName=Emissive - HasShadows?, sliderMax=1}) {desc=Add fake shadow to emissive}
   B=Multiply(A=@n63, B=Subtract(A=@n64, B=@n65))
  )
  n67 = VectorParameter() {group=03 Emissive, parameterName=Emissive - Color, defaultValue=<rgba 1 1 1 1>}
  NamedRerouteDeclaration(None=LinearInterpolate(A=Multiply(A=@n46.RGB, B=@n47) {constB=4096}, B=Multiply(A=@n66, B=@n67.RGB), Alpha=NamedRerouteUsage())) {name=Emissive, nodeColor=<rgba 1 0.5916 0 1>}
  NamedRerouteDeclaration(None=@n13.Emissive) {name=DP Emissive, nodeColor=<rgba 0.7344 0 1 1>}
  NamedRerouteDeclaration(None=@n13.Metallic) {name=DP Metallic, nodeColor=<rgba 0.7344 0 1 1>}
  NamedRerouteDeclaration(None=@n13.Dissolve) {name=DP Dissolve, nodeColor=<rgba 0.7344 0 1 1>}
  NamedRerouteUsage()
  NamedRerouteUsage()
  NamedRerouteUsage()
  NamedRerouteUsage()
  NamedRerouteUsage()
  NamedRerouteUsage()
  NamedRerouteUsage()
  NamedRerouteUsage()
  NamedRerouteUsage()
  Subtract(A=@n63, B=@n64)
 ]
}
