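# Translucent material for the FSskull dealership set's skeletal skull, producing a glowing, fiery look with painted color variants.
# Two panning noises, distorted by panning normals, form a fire pattern that modulates rim glow and a view-angle and local-position dissolve mask.
# Vertex colors separate eyes (red), teeth (green) and noise influence (blue).
# IsFS switches color sets and emission blending before Fresnel and brightness scale the emissive.
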Material /DealershipSharedLibrary/Assets/Sets/FSskull/Materials/M_DLR_FSskull
graph {
 properties {
  allowTranslucentCustomDepthWrites=true
  bAutomaticallySetUsageInEditor=false
  bUsedWithMeshDeformer=true
  bUsedWithSkeletalMesh=true
  blendMode=BLEND_Translucent
  usedShadingModels=Default Lit
 }
 nodes [
  n1 = VectorParameter() {parameterName=Color, defaultValue=<rgba 2 0.8 0.5 1>}
  n2 = VectorParameter() {parameterName=EyesMouth_Color2, defaultValue=<rgba 2 0.4409 0 1>}
  n3 = AppendVector(A=@n2.RGB, B=@n2.RGB)
  n4 = MaterialFunctionCall!(
   Primary Color=If(A=ScalarParameter() {defaultValue=1, parameterName=IsFS}, "A > B"=AppendVector(A=@n1.RGB, B=@n1.RGB), "A == B"=@n3, "A < B"=@n3)
   Tertiary Color
   "Secondary  Color"
  ) {materialFunction=/DealershipSharedLibrary/FX/Materials/MaterialFunctions/MF_DLR_PaintedVariantsData}
  n5 = VertexColor()
  n6 = NamedRerouteUsage()
  n7 = Subtract(A=DotProduct(A=CameraVectorWS(), B=VertexNormalWS()), B=Multiply(A=@n5.B, B=@n6))
  n8 = Saturate(None=@n7)
  n9 = Add(A=Power(Base=OneMinus(None=@n8), Exp=Constant() {r=10}), B=Saturate(None=Power(Base=@n6) {constExponent=4}))
  n10 = Multiply(
   A=Multiply(A=LinearInterpolate(A=@n4.Primary RGB, Alpha=@n5.B) {constB=0}, B=Saturate(None=@n9))
   B=ScalarParameter() {defaultValue=1, parameterName=BrightnessTexure}
  )
  n11 = Add(
   A=Multiply(A=@n5.R, B=ScalarParameter() {defaultValue=8.4, parameterName=EyeEmission})
   B=Multiply(A=@n5.G, B=ScalarParameter() {defaultValue=6, parameterName=TeethEmission})
  )
  n12 = ScalarParameter() {defaultValue=1, parameterName=EyesandTeethEmissionMultiplier}
  n13 = Multiply(A=Desaturation(None=Multiply(A=@n11, B=@n12)), B=@n4.Primary RGB)
  n14 = LinearInterpolate(A=@n10, B=@n13, Alpha=Saturate(None=Add(A=@n5.R, B=@n5.R)))
  n15 = If(A=ScalarParameter() {defaultValue=1, parameterName=IsFS}, "A > B"=Add(A=@n10, B=@n13), "A == B"=@n14, "A < B"=@n14)
  n16 = ScalarParameter() {defaultValue=15, group=Refraction, parameterName=BrightnessOverall}
  n17 = ScalarParameter() {defaultValue=0.34, parameterName=FresnelExponent, sortPriority=28}
  n18 = VertexNormalWS()
  n19 = Multiply(
   A=Power(Base=Fresnel(ExponentIn=@n17, Normal=@n18) {baseReflectFraction=0}, Exp=ScalarParameter() {defaultValue=3.5, parameterName=FresnelPow})
   B=ScalarParameter() {defaultValue=4, parameterName=FresnelMult}
  )
  MaterialFunctionCall(Emissive=Multiply(A=Multiply(A=@n15, B=@n16), B=Multiply(A=@n19, B=@n4.Primary RGB))) {materialFunction=&MF_FX_EmissiveAdjustments}
  n20 = Multiply(
   A=Multiply(A=TextureCoordinate() {coordinateIndex=1}, B=Constant2Vector() {g=0.75, r=1})
   B=ScalarParameter() {defaultValue=0.35, parameterName=FireTiling}
  )
  n21 = VectorParameter() {
   parameterName=Texture Panner
   channelNames {a=Texture 2 Tile V, b=Texture 2 Tile U, g=Texture 1 Tile V, r=Texture 1 Tile U}
   defaultValue=<rgba -0.05 0.15 0.05 0.1>
  }
  n22 = Panner(Coordinate=@n20, Speed=AppendVector(A=@n21.Texture 1 Tile U, B=@n21.Texture 1 Tile U))
  n23 = TextureSampleParameter2D(UVs=@n22) {
   group=05 Normal
   parameterName=Normal - Texture
   samplerType=SAMPLERTYPE_Normal
   sortPriority=0
   texture=/Game/Packages/DS_Fortnite_TRV_Terrain/Textures/T_TRV_SwampWater_01_N
  }
  n24 = Panner(Coordinate=@n20, Speed=AppendVector(A=@n21.Texture 2 Tile U, B=@n21.Texture 2 Tile U))
  n25 = TextureSampleParameter2D(UVs=@n24) {
   group=05 Normal
   parameterName=Normal - Texture
   samplerType=SAMPLERTYPE_Normal
   sortPriority=0
   texture=&T_TRV_SwampWater_01_N
  }
  n26 = ComponentMask(None=Add(A=@n23.RGB, B=@n25.RGB)) {g=true, r=true}
  n27 = Add(A=@n22, B=@n26)
  n28 = Add(A=@n24, B=@n26)
  NamedRerouteDeclaration(None=Add(
    A=TextureSampleParameter2D(UVs=@n27) {group=Texture 2, parameterName=Texture 2, texture=/Game/Effects/Fort_Effects/Textures/Tile/T_GaussianNoise_01_M}.R
    B=TextureSampleParameter2D(UVs=@n28) {group=Texture 1, parameterName=Texture 1, texture=/Game/Global/Textures/Default/Noise/T_Tiling_Noise_01}.R
   )) {name=Base Texture, nodeColor=<rgba 0.0078 1 0 1>}
  n29 = ScalarParameter() {defaultValue=0.5467, parameterName=Dissolve}
  n30 = Subtract(
   A=Add(A=@n7, B=ScalarParameter() {defaultValue=0.9017, parameterName=Falloff})
   B=LinearInterpolate(A=Constant() {r=1}, B=Constant() {r=2.2}, Alpha=Saturate(None=@n29))
  )
  n31 = Constant() {r=2.2893}
  n32 = Constant() {r=128}
  n33 = WorldPosition()
  n34 = ObjectPositionWS()
  n35 = Transform(None=Add(A=Multiply(A=@n6, B=@n32), B=Subtract(A=@n33.XYZ, B=@n34))) {transformSourceType=TRANSFORMSOURCE_World, transformType=TRANSFORM_Local}
  n36 = Multiply(A=Subtract(A=ComponentMask(None=@n35) {b=true, r=true}, B=ScalarParameter() {defaultValue=0.9, parameterName=ForwardDissolve}))
  Saturate(None=Multiply(A=Multiply(A=@n30, B=@n31), B=Saturate(None=@n36)))
  VectorParameter() {parameterName=PaintedColor2, defaultValue=<rgba 0.0104 0.0065 0.0065 1>}
 ]
}
