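# Physics asset for the Skull_Anim skeletal mesh in the FSskull set of the dealership shared library.
# It defines a single capsule body on the joint1 bone, offset and rotated from the bone's origin.
# It has no constraints, so the skull's collision is one shape with no jointed ragdoll parts.
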PhysicsAsset /DealershipSharedLibrary/Assets/Sets/FSskull/Meshes/Skull_Anim_PhysicsAsset
bodies {joint1 [{shapeType=Capsule, length=27.1611, radius=18.2007, center=<6.6337 -1.4595 1.384>, rotation=<rot -0.5038 -90 -39.2882>}]}
constraints []
